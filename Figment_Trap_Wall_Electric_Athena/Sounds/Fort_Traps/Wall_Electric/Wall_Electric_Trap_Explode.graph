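# Sound cue for the explosion or destruction of the electric wall trap in the Figment trap plugin.
# A Random node picks one of two equally weighted destroy waves and passes it through a Modulator.
# It uses the shared Fort_Traps attenuation and sound class, plays at reduced volume, and overrides concurrency to prevent new instances once the limit is reached.
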SoundCue /Figment_Trap_Wall_Electric_Athena/Sounds/Fort_Traps/Wall_Electric/Wall_Electric_Trap_Explode
root=Modulator_4
nodes [
 {id=Modulator_4, children [Random_17]}
 {id=Random_17, children [WavePlayer_62, WavePlayer_63], properties {weights [1,1]}}
 {
  id=WavePlayer_62
  properties {soundWaveAssetPtr=&/WallTrap_Electric_Destroy_001}
 }
 {
  id=WavePlayer_63
  properties {soundWaveAssetPtr=&/WallTrap_Electric_Destroy_002}
 }
]
properties {
 attenuationSettings=&Fort_Traps_Att
 bOverrideConcurrency=true
 maxDistance=6000
 soundClassObject=&Fort_Traps
 volumeMultiplier=0.5
 concurrencyOverrides {resolutionRule=PreventNew}
}
