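# Sound cue for the electric wall trap's zap, using trap sound class and attenuation with a concurrency override that prevents new instances at the limit.
# A mixer layers an immediate, modulated random pick from three wall-trap fire waves with a second layer delayed by a fixed interval.
# The delayed layer reuses ceiling-trap ready sounds, with a quality-level switch choosing either a random pair or a single "opt" wave.
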SoundCue /Figment_Trap_Wall_Electric_Athena/Sounds/Fort_Traps/Wall_Electric/Wall_Electric_Trap_Zap
root=Mixer_19
nodes [
 {id=Delay_10, children [QualityLevel_0], properties {delayMax=0.7, delayMin=0.7}}
 {id=Mixer_19, children [Modulator_9,Delay_10], properties {inputVolume [1,1]}}
 {id=Modulator_9, children [Random_18]}
 {id=QualityLevel_0, children [Random_44,Random_44,WavePlayer_0,WavePlayer_0]}
 {id=Random_18, children [WavePlayer_65, WavePlayer_66, WavePlayer_68], properties {weights [1,1,1]}}
 {id=Random_44, children [WavePlayer_152, WavePlayer_153], properties {weights [1,1]}}
 {id=WavePlayer_0, properties {soundWaveAssetPtr=/FNGameplayCues/Sounds/Fort_Traps/Electric_Trap/CeilingTrap_Electric_Ready_002_opt}}
 {id=WavePlayer_152, properties {soundWaveAssetPtr=/Game/Sounds/Fort_Traps/Electric_Trap/CeilingTrap_Electric_Ready_001}}
 {id=WavePlayer_153, properties {soundWaveAssetPtr=/FNGameplayCues/Sounds/Fort_Traps/Electric_Trap/CeilingTrap_Electric_Ready_002}}
 {id=WavePlayer_65, properties {soundWaveAssetPtr=/Game/Sounds/Fort_Traps/Wall_Electric/WallTrap_Electric_Fire_004}}
 {id=WavePlayer_66, properties {soundWaveAssetPtr=/Game/Sounds/Fort_Traps/Wall_Electric/WallTrap_Electric_Fire_001}}
 {id=WavePlayer_68, properties {soundWaveAssetPtr=/Game/Sounds/Fort_Traps/Wall_Electric/WallTrap_Electric_Fire_003}}
]
properties {
 attenuationSettings=&Fort_Traps_Att
 bOverrideConcurrency=true
 maxDistance=6000
 soundClassObject=&Fort_Traps
 volumeMultiplier=0.5
 attenuationOverrides {falloffDistance=3200, attenuationShapeExtents {x=800}}
 concurrencyOverrides {resolutionRule=PreventNew}
}
